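# Material function for the MeteorWoman suit that projects a scrolling noise texture in pre-skinned local space, so the pattern follows the animated mesh.
# Scaled position plus fractional time-based offset supplies planar UVs.
# Contrast-adjusted absolute normal Y blends YZ and XZ samples into the output.
# A further XY blend by normal Z is built but left unconnected, so the result uses only two of the three planes.
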MaterialFunction /BRCosmetics/Characters/Player/Female/Medium/Bodies/F_MED_MeteorWoman_Suit/Materials/MF_Meteorwoman_Triplanar
graph {nodes [
  n1 = PreSkinnedPosition()
  n2 = Divide(
   A=FunctionInput(Preview=VertexInterpolator(VS=@n1).PS) {bUsePreviewValueAsDefault=true, inputName=Local Position}
   B=FunctionInput(Preview=Constant() {r=40}) {inputName=Triplanner ConstantScaleDivisor, inputType=FunctionInput_Scalar}
  )
  n3 = Constant3Vector() {constant=<rgba 0.1 0.1 0.1 1>}
  n4 = Frac(None=Multiply(A=Time(), B=FunctionInput(Preview=@n3) {inputName=Triplanner Speed}))
  n5 = Add(A=@n2, B=@n4)
  n6 = TextureObject() {
   samplerType=SAMPLERTYPE_Masks
   texture=/BRCosmetics/Characters/Player/Female/Medium/Bodies/F_MED_MeteorWoman_Suit/Textures/T_F_MED_MeteorWoman_Noise
  }
  n7 = FunctionInput(Preview=@n6) {inputName=Texture, inputType=FunctionInput_Texture2D}
  n8 = Add(A=@n2, B=@n4)
  n9 = FunctionInput(Preview=VertexInterpolator(VS=PreSkinnedNormal()).PS) {bUsePreviewValueAsDefault=true, inputName=Local Normal}
  n10 = FunctionInput(Preview=Constant() {r=2}) {inputName=TriPlanner AxisBlendStrength, inputType=FunctionInput_Scalar}
  n11 = MaterialFunctionCall(In=Abs(None=ComponentMask(None=@n9) {g=true}), Contrast=@n10) {materialFunction=&CheapContrast}
  n12 = LinearInterpolate(
   A=TextureSample(UVs=ComponentMask(None=@n5) {b=true, g=true}, Tex=@n7) {samplerType=SAMPLERTYPE_Masks, texture=/Game/Weapons/WeaponSkins/Textures/T_Swirly_Noise_highContrast}.RGB
   B=TextureSample(UVs=ComponentMask(None=@n8) {b=true, r=true}, Tex=@n7) {samplerType=SAMPLERTYPE_Masks, texture=&T_Swirly_Noise_highContrast}.RGB
   Alpha=Saturate(None=ComponentMask(None=@n11.Result) {g=true})
  )
  FunctionOutput(None=@n12)
  n13 = Add(A=@n2, B=@n4)
  n14 = MaterialFunctionCall(In=Abs(None=ComponentMask(None=@n9) {b=true}), Contrast=@n10) {materialFunction=&CheapContrast}
  LinearInterpolate(
   A=@n12
   B=TextureSample(UVs=ComponentMask(None=@n13) {g=true, r=true}, Tex=@n7) {samplerType=SAMPLERTYPE_Masks, texture=&T_Swirly_Noise_highContrast}.RGB
   Alpha=Saturate(None=ComponentMask(None=@n14.Result) {b=true})
  )
 ]}
