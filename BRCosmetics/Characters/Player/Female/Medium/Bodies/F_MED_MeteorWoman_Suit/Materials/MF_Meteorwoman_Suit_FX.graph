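# Material function for the MeteorWoman suit, adding masked metal LUT, thin film and hit-reactive emissive effects to input material attributes.
# Curve-atlas color, driven by a Fresnel or height gradient, combines with triplanar circuitry and flow noise in pre-skinned space.
# Blueprint-fed hit time and position drive expanding SphereMask hit and delayed restore waves.
# UseEmissive off returns the raw input, bypassing every effect; the panning emissive branch is unconnected.
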MaterialFunction /BRCosmetics/Characters/Player/Female/Medium/Bodies/F_MED_MeteorWoman_Suit/Materials/MF_Meteorwoman_Suit_FX
graph {nodes [
  n1 = TextureObjectParameter() {
   group=Skin FX
   parameterName=SkinFX_Mask
   samplerType=SAMPLERTYPE_LinearColor
   sortPriority=0
   texture=/BRCosmetics/Characters/Player/Female/Medium/Bodies/F_MED_MeteorWoman_Suit/Textures/T_F_MED_MeteorWoman_Suit_Body_FX
  }
  n2 = TextureSample(Tex=@n1)
  n3 = FunctionInput() {inputName=Material, inputType=FunctionInput_MaterialAttributes}
  n4 = StaticSwitchParameter(
   True=MaterialFunctionCall(Mask=ChannelMaskParameter(None=@n2.RGBA) {group=Skin FX, parameterName=MetalLUTChannel}.R, Materials=@n3) {materialFunction=&MF_MetalsLUT}.Result
   False=@n3
  ) {group=Skin FX, parameterName=UseMetalLUT, sortPriority=3}
  n5 = ChannelMaskParameter(None=@n2.RGBA) {group=Skin FX, maskChannel=Alpha, parameterName=ThinFilmChannel, defaultValue=<rgba 0 0 0 1>}
  n6 = GetMaterialAttributes(None=StaticSwitchParameter(
    True=MaterialFunctionCall!(Alpha=@n5.R, Material Attributes=@n4) {materialFunction=/Game/Weapons/WeaponSkins/MaterialFunction/ThinFilmFunction}.Result
    False=@n4
   ) {group=Skin FX, parameterName=Use Thin Film}) {attributeGetTypes [#1, #2]}
  n7 = NamedRerouteUsage()
  n8 = LinearInterpolate(
   A=@n7
   B=Multiply(A=NamedRerouteUsage(), B=@n7)
   Alpha=ScalarParameter() {defaultValue=1, group=2 Meteorwomen Emissive, parameterName=FresnelIntensity}
  )
  n9 = ScalarParameter() {defaultValue=1, group=2 Meteorwomen Emissive, parameterName=EmissiveIntensity, sortPriority=1}
  n10 = NamedRerouteUsage()
  n11 = Add(
   A=LinearInterpolate(A=@n6.EmissiveColor, B=Multiply(A=@n8, B=@n9) {constB=0.2}, Alpha=@n10)
   B=Multiply(A=NamedRerouteUsage(), B=VectorParameter() {parameterName=SecondEmissiveColor, defaultValue=<rgba 0.9688 0 0 1>}.RGB)
  )
  n12 = SetMaterialAttributes!(MaterialAttributes=@n6.MaterialAttributes, EmissiveColor=@n11, BaseColor=Multiply(A=@n6.BaseColor, B=OneMinus(None=@n10))) {attributeSetTypes [#1, #2]}
  n13 = SetMaterialAttributes!(MaterialAttributes=@n6.MaterialAttributes, EmissiveColor=Add(A=@n11, B=@n6.EmissiveColor)) {attributeSetTypes [#1]}
  n14 = ScalarParameter() {group=2 Meteorwomen Emissive, parameterName=AdditiveMateiral_Blend, sliderMax=1, sortPriority=40}
  FunctionOutput(None=StaticSwitchParameter(True=BlendMaterialAttributes(A=@n12, B=@n13, Alpha=@n14), False=@n3) {group=Skin FX, parameterName=UseEmissive, sortPriority=3})
  n15 = NamedRerouteUsage()
  n16 = Constant() {r=0.1}
  n17 = Constant() {r=100}
  n18 = TextureObject() {
   samplerType=SAMPLERTYPE_Masks
   texture=/BRCosmetics/Characters/Player/Female/Medium/Bodies/F_MED_MeteorWoman_Suit/Textures/T_F_MED_MeteorWoman_Noise
  }
  n19 = NamedRerouteUsage()
  n20 = MaterialFunctionCall!(
   Local Position=@n15
   Triplanner Speed=Constant3Vector()
   TriPlanner AxisBlendStrength=@n16
   Triplanner ConstantScaleDivisor=@n17
   Texture=@n18
   Local Normal=@n19
  ) {materialFunction=&/MF_Meteorwoman_Triplanar}
  MaterialFunctionCall(Float4=@n20.Result) {materialFunction=&BreakOutFloat4Components}
  NamedRerouteDeclaration(None=ChannelMaskParameter(None=@n2.RGBA) {group=Skin FX, maskChannel=Green, parameterName=EmissiveChannel, defaultValue=<rgba 0 1 0 0>}.R) {name=EmissiveMask, nodeColor=<rgba 1 0 0.4219 1>}
  n21 = PreSkinnedPosition()
  NamedRerouteDeclaration(None=FunctionInput(Preview=VertexInterpolator(VS=@n21).PS) {bUsePreviewValueAsDefault=true, inputName=PreSkinnedLocalPosition, sortPriority=4}) {name=PreskinnedLocalPosition, nodeColor=<rgba 0 1 0.1501 1>}
  n22 = Power(Base=NamedRerouteUsage()) {constExponent=0.4}
  n23 = OneMinus(None=ScalarParameter() {defaultValue=1, group=2 Meteorwomen Emissive, parameterName=FresnelGradientMovement})
  n24 = Power(
   Base=OneMinus(None=Fresnel() {exponent=1})
   Exp=ScalarParameter() {defaultValue=1.5, group=2 Meteorwomen Emissive, parameterName=FresnelGradientPower}
  ) {constExponent=1.5}
  n25 = Saturate(None=Multiply(A=LinearInterpolate(A=@n22, Alpha=@n23), B=Multiply(A=@n24) {constB=3}))
  n26 = ScalarParameter() {defaultValue=1, group=2 Meteorwomen Emissive, parameterName=FresnelGradientIntensity}
  n27 = NamedRerouteUsage()
  n28 = PreSkinnedLocalBounds()
  n29 = ComponentMask(None=Divide(A=Subtract(A=@n27, B=@n28.Min), B=@n28.Extents)) {b=true}
  n30 = CurveAtlasRowParameter(CurveTime=StaticSwitchParameter(True=LinearInterpolate(B=@n25, Alpha=@n26) {constA=1}, False=Saturate(None=@n29)) {group=2 Meteorwomen Emissive, parameterName=UseFresnelGradient}) {
   atlas=/BRCosmetics/Characters/Player/Female/Medium/Bodies/F_MED_MeteorWoman_Suit/Textures/EmissiveCurves/CA_Meteorwoman_Suit
   curve=/BRCosmetics/Characters/Player/Female/Medium/Bodies/F_MED_MeteorWoman_Suit/Textures/EmissiveCurves/C_Meteorwoman_Suit_03
   defaultValue=2
   group=2 Meteorwomen Emissive
   parameterName=MeteorwomanEmissiveCurve
   sortPriority=0
  }
  n31 = NamedRerouteDeclaration(None=@n30) {name=VarientColor, nodeColor=<rgba 0.3359 0 1 1>}
  n32 = ScalarParameter() {defaultValue=1, group=2 Meteorwomen Emissive, parameterName=BaseEmissiveIntensity}
  n33 = ScalarParameter() {defaultValue=0.2, group=2 Meteorwomen Emissive, parameterName=DimEmissiveIntensity, sliderMax=1}
  n34 = LinearInterpolate(
   A=StaticSwitchParameter(True=@n32, False=StaticSwitchParameter(True=@n33, False=@n32) {group=1 Meteorwomen Reactivity, parameterName=DarkenWhenHit}) {group=2 Meteorwomen Emissive, parameterName=EmissiveOn}
   B=@n32
   Alpha=NamedRerouteUsage()
  )
  n35 = Multiply(A=NamedRerouteUsage(), B=NamedRerouteUsage())
  n36 = ScalarParameter() {defaultValue=7, group=2 Meteorwomen Emissive, parameterName=AmbientIntensity}
  n37 = OneMinus(None=Saturate(None=Multiply(A=@n35, B=@n36) {constB=0.5}))
  n38 = LinearInterpolate(
   A=Multiply(A=@n31, B=@n34)
   B=Multiply(A=NamedRerouteDeclaration(None=@n37) {name=Ambient, nodeColor=<rgba 0.4062 0 1 1>}, B=@n31)
   Alpha=@n34
  )
  n39 = ScalarParameter() {defaultValue=1, group=2 Meteorwomen Emissive, parameterName=DefaultEmissiveAmount, sliderMax=1, sortPriority=40}
  n40 = VectorParameter() {group=1 Meteorwomen Reactivity, parameterName=HitColor, sortPriority=1, defaultValue=<rgba 1 0.02 0 1>}
  n41 = Multiply(A=@n40.RGB, B=ScalarParameter() {defaultValue=3.5, group=1 Meteorwomen Reactivity, parameterName=HitIntensity, sortPriority=0})
  n42 = NamedRerouteUsage()
  NamedRerouteDeclaration(None=LinearInterpolate(A=Multiply(A=@n38, B=@n39), B=Multiply(A=@n41, B=@n42), Alpha=@n42)) {name=EmissiveColor, nodeColor=<rgba 1 0 0.6797 1>}
  n43 = ScalarParameter() {group=1 Meteorwomen Reactivity, parameterName=Hit Min}
  n44 = ScalarParameter() {defaultValue=200, group=1 Meteorwomen Reactivity, parameterName=Hit Max}
  n45 = ScalarParameter() {group=1 Meteorwomen Reactivity, parameterName=GameTimeFromBP}
  n46 = Time()
  n47 = Subtract(A=@n46, B=@n45) {desc=Elapsed time}
  n48 = ScalarParameter() {defaultValue=1.5, group=1 Meteorwomen Reactivity, parameterName=HitLength}
  n49 = Divide(A=@n47, B=@n48)
  n50 = NamedRerouteDeclaration(None=Multiply(A=ScalarParameter() {group=1 Meteorwomen Reactivity, parameterName=HitAmountMulti}, B=Saturate(None=@n49))) {name=HitAmount, nodeColor=<rgba 1 0.7031 0 1>}
  n51 = Multiply(A=Sign(None=Subtract(A=@n48, B=@n47) {desc="> length ?"}), B=Sign(None=Subtract(A=@n47) {constA=0, constB=0, desc="<=  0 ?"}))
  NamedRerouteDeclaration(None=Multiply(A=LinearInterpolate(A=@n43, B=@n44, Alpha=@n50), B=Saturate(None=@n51))) {name=HitSize, nodeColor=<rgba 1 0.375 0 1>}
  n52 = ScalarParameter() {defaultValue=0.1, group=2 Meteorwomen Emissive, parameterName=EmissiveFresnelExp}
  n53 = VertexNormalWS()
  NamedRerouteDeclaration(None=OneMinus(None=Fresnel(ExponentIn=@n52, Normal=@n53))) {name=EmissiveFresnel, nodeColor=<rgba 1 0.8906 0 1>}
  n54 = VectorParameter() {group=1 Meteorwomen Reactivity, parameterName=HitVector, defaultValue=<rgba 0 0 0 1>}
  n55 = SphereMask(A=NamedRerouteUsage(), B=NamedRerouteDeclaration(None=@n54.RGB) {name=HitVector, nodeColor=<rgba 1 0.5625 0 1>}, Radius=NamedRerouteUsage()) {attenuationRadius=90, desc=Outter ring, hardnessPercent=0}
  n56 = Power(Base=@n55, Exp=Divide(B=ScalarParameter() {defaultValue=100, parameterName=Edge Hardness})) {constExponent=0.01}
  n57 = NamedRerouteUsage()
  n58 = Multiply(
   A=Power(Base=@n55, Exp=Constant() {r=2}) {constExponent=0.1}
   B=Multiply(A=NamedRerouteUsage(), B=ScalarParameter() {defaultValue=1.25, group=1 Meteorwomen Reactivity, parameterName=RingFadeStrength})
  )
  n59 = Saturate(None=Add(A=Multiply(A=@n57, B=@n58), B=@n58))
  n60 = Saturate(None=Power(Base=Subtract(A=@n56, B=@n59), Exp=ScalarParameter() {defaultValue=3, group=1 Meteorwomen Reactivity, parameterName=RingPower}))
  n61 = NamedRerouteUsage()
  n62 = Constant() {r=1}
  n63 = Multiply(A=Multiply(A=NamedRerouteUsage(), B=LinearInterpolate(B=@n61, Alpha=@n62) {constA=1}), B=Constant() {r=0.3})
  n64 = NamedRerouteUsage()
  n65 = ScalarParameter() {defaultValue=0.1, group=1 Meteorwomen Reactivity, parameterName=Noise2 Intensity}
  n66 = Multiply(A=Subtract(A=Subtract(A=@n60, B=@n63) {desc=Outter Ring}, B=Multiply(A=@n64, B=@n65) {constB=0.2}) {constB=0}, B=Constant() {r=1.8}) {constB=3.5}
  n67 = OneMinus(None=Power(Base=NamedRerouteUsage()) {constExponent=5})
  NamedRerouteDeclaration(None=Saturate(None=Multiply(A=@n66, B=@n67))) {name=HitWave, nodeColor=<rgba 1 0.6797 0 1>}
  n68 = Constant() {r=500}
  n69 = Subtract(A=@n46, B=Add(A=@n45, B=ScalarParameter() {defaultValue=0.5, group=1 Meteorwomen Reactivity, parameterName=RestoreDelay}))
  n70 = ScalarParameter() {defaultValue=2.5, group=1 Meteorwomen Reactivity, parameterName=RestoreLength}
  n71 = NamedRerouteDeclaration(None=Saturate(None=Divide(A=@n69, B=@n70) {constB=0.5})) {name=RestoreAmount, nodeColor=<rgba 1 0.7734 0 1>}
  n72 = SphereMask(
   A=NamedRerouteUsage()
   B=NamedRerouteUsage()
   Radius=NamedRerouteDeclaration(None=LinearInterpolate(B=@n68, Alpha=@n71) {constB=150}) {name=RestoreSize, nodeColor=<rgba 0.3125 0 1 1>}
  ) {attenuationRadius=90, desc=Outter ring, hardnessPercent=0}
  n73 = Power(Base=@n72) {constExponent=1}
  n74 = Subtract(A=@n72, B=Add(A=NamedRerouteUsage(), B=NamedRerouteUsage()))
  n75 = OneMinus(None=@n74)
  n76 = Saturate(None=Power(Base=@n72) {constExponent=20})
  n77 = Multiply(A=Add(A=Multiply(A=@n73, B=@n75), B=Add(A=@n76, B=@n74)), B=Constant() {r=0.5})
  n78 = NamedRerouteUsage()
  NamedRerouteDeclaration(None=Saturate(None=LinearInterpolate(A=@n77, Alpha=@n78))) {name=RestoreWave, nodeColor=<rgba 1 0.4922 0 1>}
  n79 = ChannelMaskParameter(None=@n2.RGBA) {group=2 Meteorwomen Emissive, maskChannel=Blue, parameterName=SecondEmissiveChannel, defaultValue=<rgba 0 0 1 0>}
  NamedRerouteDeclaration(None=@n79.R) {name=SecondEmissiveMask, nodeColor=<rgba 0.5703 1 0 1>}
  n80 = Constant3Vector() {constant=<rgba 0.087 -0.08 0.123 0>}
  n81 = Constant3Vector() {constant=<rgba -0.112 0.04 -0.0498 0>}
  n82 = Multiply(
   A=MaterialFunctionCall(Float3=MaterialFunctionCall(
     Local Position=@n15
     Triplanner Speed=@n80
     TriPlanner AxisBlendStrength=@n16
     Triplanner ConstantScaleDivisor=@n17
     Texture=@n18
     Local Normal=@n19
    ) {materialFunction=&MF_Meteorwoman_Triplanar}.Result) {materialFunction=&BreakOutFloat3Components}.R
   B=MaterialFunctionCall(Float3=MaterialFunctionCall(
     Local Position=@n15
     Triplanner Speed=@n81
     TriPlanner AxisBlendStrength=@n16
     Triplanner ConstantScaleDivisor=@n17
     Texture=@n18
     Local Normal=@n19
    ) {materialFunction=&MF_Meteorwoman_Triplanar}.Result) {materialFunction=&BreakOutFloat3Components}.R
  )
  NamedRerouteDeclaration(None=Add(A=Power(Base=@n82) {constExponent=1.8}) {constB=0.02}) {name=Flow, nodeColor=<rgba 0.1719 0 1 1>}
  NamedRerouteDeclaration(None=ComponentMask(None=@n54.RGB) {g=true, r=true}) {name=HitPointXY, nodeColor=<rgba 0 0.1562 1 1>}
  n83 = MaterialFunctionCall(Float3=@n20.Result) {materialFunction=&BreakOutFloat3Components}
  NamedRerouteDeclaration(None=@n83.G) {name=Circuitry, nodeColor=<rgba 0.7812 1 0 1>}
  NamedRerouteDeclaration(None=@n83.B) {name=CircuitryBlurred, nodeColor=<rgba 1 0 0.5391 1>}
  n84 = PreSkinnedNormal()
  NamedRerouteDeclaration(None=FunctionInput(Preview=VertexInterpolator(VS=@n84).PS) {bUsePreviewValueAsDefault=true, inputName=PreSkinnedLocalNormal, sortPriority=5}) {name=PreskinnedLocalNormal, nodeColor=<rgba 0 1 0.7891 1>}
  n85 = TextureSample(UVs=Add(A=TextureCoordinate(), B=NamedRerouteUsage()), Tex=@n18) {
   samplerType=SAMPLERTYPE_Masks
   texture=&T_F_MED_MeteorWoman_Noise
  }
  NamedRerouteDeclaration(None=@n85.A) {name=Noise, nodeColor=<rgba 0.8984 0 1 1>}
  n86 = ChannelMaskParameter() {group=Skin FX, maskChannel=Blue, parameterName=PanningEmissiveChannel, defaultValue=<rgba 0 0 1 0>}
  n87 = ChannelMaskParameter() {group=Skin FX, maskChannel=Blue, parameterName=PanningEmissiveUVOffsetChannel, defaultValue=<rgba 0 0 1 0>}
  StaticSwitchParameter(
   True=MaterialFunctionCall!(
    Alpha=@n86.R
    isVehicle
    Gmap=FunctionInput() {inputName=Gmap, inputType=FunctionInput_Scalar}
    UVOffset=@n87.R
    Local Normal
    Local Position
    EmissiveColorOverride
    Material=@n3
   ) {materialFunction=/Game/Weapons/WeaponSkins/MaterialFunction/PanningEmissive}.Result
   False=@n3
  ) {group=Skin FX, parameterName=Use Panning Emissive, sortPriority=2}
 ]}
